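# Translucent UI material that draws a minimap block texture, apparently for The Block, fading it by height.
# A height threshold, optionally offset by a heightmap through a static switch, is compared with the Height value from MPC_Minimap over a smooth transition distance.
# The clamped result scales texture alpha, hiding the block once that Height reaches the threshold.
# A Debug switch tints faded areas with a solid color.
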Material /Figment_S10/Athena/HUD/MiniMap/M_MinimapTheBlock
graph {
 properties {blendMode=BLEND_Translucent, materialDomain=MD_UI, pixelDepthOffsetMode=PDOM_Legacy, usedShadingModels=Default Lit}
 nodes [
  n1 = TextureSampleParameter2D() {parameterName=MinimapBlock, texture=/FigmentEnvironment/FN_6_20/Characters/Fortress_Character/Textures/AllBlack}
  n2 = ScalarParameter() {defaultValue=768, parameterName=HeightPosition}
  n3 = ScalarParameter() {defaultValue=768, parameterName=HeightPositionAddfromHeightmap}
  n4 = TextureSampleParameter2D() {parameterName=Heightmap, texture=&AllBlack}
  n5 = StaticSwitchParameter(True=Add(A=@n2, B=Multiply(A=@n3, B=@n4.R)), False=@n2) {parameterName=UseHeightmap}
  n6 = CollectionParameter() {collection=/Game/Athena/HUD/MiniMap/MPC_Minimap, parameterName=Height}
  n7 = Saturate(None=Divide(A=Subtract(A=@n5, B=@n6), B=ScalarParameter() {defaultValue=300, parameterName=SmoothTransition}))
  Multiply(A=@n1.A, B=@n7)
  StaticSwitchParameter(True=LinearInterpolate(A=VectorParameter() {defaultValue=<rgba 1 0.0888 0 1>}.RGB, B=@n1.RGB, Alpha=@n7), False=@n1.RGB) {parameterName=Debug}
 ]
}
